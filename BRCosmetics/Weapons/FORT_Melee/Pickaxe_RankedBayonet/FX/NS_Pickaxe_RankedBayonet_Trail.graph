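# Niagara swing trail for the Ranked Bayonet pickaxe, built from three short-lived ribbon emitters: SoftRibbon, FillTrail and FillTrail001.
# Each uses FN anim-trail modules to follow skeletal mesh bones or sockets, with FillTrail running two trails that differ in width and offset by spawn group.
# System rank values shift each ribbon's hue, saturation and value, and time-of-day colors tint them.
# Curves fade alpha, narrow width and drive dynamic material parameters.
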NiagaraSystem /BRCosmetics/Weapons/FORT_Melee/Pickaxe_RankedBayonet/FX/NS_Pickaxe_RankedBayonet_Trail
dependencies {
 usedDataInterfaces [&NiagaraDataInterfaceColorCurve, &NiagaraDataInterfaceCurve, &NiagaraDataInterfaceSkeletalMesh]
 usedDynamicInputs [
  &MakeCustomIntFromBool
  &SetBoolByIntComparison
  &MakeIntFromFloat
  &Subtract_Float
  &Divide_Float
  &MakeFloatFromInt
  /Niagara/DynamicInputs/Subtract/Subtract_Integer
  /Niagara/DynamicInputs/Divide/Divide_Int
  &MergeHelper_FloatSwitch
  &FNTimeOfDayScalar
 ]
 usedModules [
  &self:SystemSpawnScript.SystemScriptSource.SystemScriptGraph.NiagaraNodeAssignment_0.SetVariables_#1
  &SystemState
  &self:SystemSpawnScript.SystemScriptSource.SystemScriptGraph.NiagaraNodeAssignment_1.SetVariables_#2
  &self:SystemSpawnScript.SystemScriptSource.SystemScriptGraph.NiagaraNodeAssignment_2.SetVariables_#3
  &self:SystemSpawnScript.SystemScriptSource.SystemScriptGraph.NiagaraNodeAssignment_3.SetVariables_#4
 ]
}
emitters {
 FillTrail {
  emitterSpawnScript [&FN_RibbonIDOffsetUtility()]
  emitterUpdateScript [
   &EmitterState()
   &FN_SpawnAnimTrail(Number of Anim Trails=2)
  ]
  particleSpawnScript [
   &FN_InitializeAnimTrailBoneIndecies(First Bone or Socket Index=1, Second Bone or Socket Index=0)
   &FN_InitializeAnimTrailBoneIndecies(AnimTrail Selection=1, First Bone or Socket Index=1, Second Bone or Socket Index=0) {module=FN_InitializeAnimTrailBoneIndecies001}
   &FN_InitializeAnimTrail(
    AnimTrailWidth=&MaskFloatBySpawnGroup(Masked Float=1, Passthrough Float=1.5)
    Color=&HueShiftLinearColor(Hue=$System.RankHueShift, Saturation=$System.RankSaturation, Value=$System.RankValue, Linear Color=<rgba 0.4531 0.0885 0.0001 1.2>)
    Lifetime=0.3
    Position Offset=&MaskVectorBySpawnGroup(
     Masked Vector=<0 0 0>
     Passthrough Vector=&TransformVector(Vector=<0 10 -13>, Destination Space=::Simulation)
    )
    SkeletalMesh=$System.Skeletal Mesh
   )
  ]
  particleUpdateScript [
   &SolveForcesAndVelocity(Clamp Velocity=false, Limit Acceleration=false)
   &FN_UpdateAnimTrail(Locked Particle RibbonWidth=1, SkeletalMesh=$System.Skeletal Mesh) {module=FN_UpdateAnimTrail001}
   &ScaleColor(
    Scale Alpha=&FloatFromCurve(
     FloatCurve=DataInterface() {propertyValues {Curve {keys [
         {interpMode=RCIM_Cubic, leaveTangent=-18.1448, tangentMode=RCTM_Break, time=0, value=1}
         {arriveTangent=-0.1476, tangentMode=RCTM_User, time=0.13, value=0}
        ]}}}
     Scale Curve=0.75
    )
    Scale RGB=&MakeVectorFromLinearColorRGB(Color=$System.ToDColor)
    ScaleA=true
    ScaleRGB=true
   )
   &ScaleRibbonWidth(Ribbon Width Scale=&FloatFromCurve(FloatCurve=DataInterface() {propertyValues {Curve {keys [{interpMode=RCIM_Cubic, leaveTangent=-9.53, tangentMode=RCTM_Break, time=0.2, value=1}, <at 0.45 0.1>]}}}))
   &DynamicMaterialParameters(
    Index 0 Param 1=&FloatFromCurve(FloatCurve=DataInterface() {propertyValues {Curve {keys [<at 0 1>,<at 1 0>]}}})
    Index 0 Param 2=0.3
    Index 0 Param 3=2
    Index 0 Param 4=1
    Module.Index0Param0Write=true
    Module.Index0Param1Write=true
    Module.Index0Param2Write=true
    Module.Index0Param3Write=true
   )
  ]
  rendererClasses [&NiagaraRibbonRendererProperties]
  renderers [{rendererClass=&NiagaraRibbonRendererProperties}]
 }
 FillTrail001 {
  emitterSpawnScript [&FN_RibbonIDOffsetUtility()]
  emitterUpdateScript [
   &EmitterState()
   &FN_SpawnAnimTrail()
  ]
  particleSpawnScript [
   &FN_InitializeAnimTrailBoneIndecies(First Bone or Socket Index=1, Second Bone or Socket Index=0)
   &FN_InitializeAnimTrail(
    Color=&HueShiftLinearColor(Hue=$System.RankHueShift, Saturation=$System.RankSaturation, Value=$System.RankValue, Linear Color=<rgba 0.5 0.3104 0.175 0.5>)
    Lifetime=0.3
    SkeletalMesh=$System.Skeletal Mesh
   )
   &Color(Color=&MultiplyLinearColor(
     Linear Color=$Particles.Initial.Color
     Scale Factor=&MakeVector4_FromLinearColor(LinearColor=&FNTimeOfDayColor(
       Night Color=<rgba 0.5387 0.487 1 1>
       Morning Color=<rgba 0.8728 0.9258 1 1>
       Day Color=<rgba 1 1 1 1>
       Evening Color=<rgba 0.8728 0.9258 1 1>
      ))
    ))
  ]
  particleUpdateScript [
   &SolveForcesAndVelocity(Clamp Velocity=false, Limit Acceleration=false)
   &FN_UpdateAnimTrail(Locked Particle RibbonWidth=1, SkeletalMesh=$System.Skeletal Mesh) {module=FN_UpdateAnimTrail001}
   &ScaleColor(
    Scale Alpha=&FloatFromCurve(
     FloatCurve=DataInterface() {propertyValues {Curve {keys [
         {interpMode=RCIM_Cubic, leaveTangent=-19.5372, tangentMode=RCTM_Break, time=0, value=1}
         {arriveTangent=-0.1476, tangentMode=RCTM_User, time=0.13, value=0}
        ]}}}
     Scale Curve=0.35
    )
    Scale RGB=&MakeVectorFromLinearColorRGB(Color=$System.ToDColor)
    ScaleA=true
    ScaleRGB=true
   )
   &ScaleRibbonWidth(Ribbon Width Scale=&FloatFromCurve(FloatCurve=DataInterface() {propertyValues {Curve {keys [{interpMode=RCIM_Cubic, leaveTangent=-5.18, tangentMode=RCTM_Break, time=0.2, value=1}, <at 0.65 0.1>]}}}))
   &DynamicMaterialParameters(
    Index 0 Param 1=&FloatFromCurve(FloatCurve=DataInterface() {propertyValues {Curve {keys [<at 0 1>,<at 1 0>]}}})
    Index 0 Param 2=0.3
    Index 0 Param 3=2
    Index 0 Param 4=1
    Module.Index0Param0Write=true
    Module.Index0Param1Write=true
    Module.Index0Param2Write=true
    Module.Index0Param3Write=true
   )
  ]
  rendererClasses [&NiagaraRibbonRendererProperties]
  renderers [{rendererClass=&NiagaraRibbonRendererProperties}]
 }
 SoftRibbon {
  emitterSpawnScript [&FN_RibbonIDOffsetUtility()]
  emitterUpdateScript [
   &EmitterState()
   &FN_SpawnAnimTrail()
  ]
  particleSpawnScript [
   &FN_InitializeAnimTrailBoneIndecies()
   &FN_InitializeAnimTrail(
    Color=&HueShiftLinearColor(Hue=$System.RankHueShift, Saturation=$System.RankSaturation, Value=$System.RankValue, Linear Color=<rgba 1.2 0.912 0.5088 0.4>)
    Lifetime=0.3
    SkeletalMesh=$System.Skeletal Mesh
   )
   &Color(Color=&MultiplyLinearColor(
     Linear Color=$Particles.Initial.Color
     Scale Factor=&MakeVector4_FromLinearColor(LinearColor=&FNTimeOfDayColor(
       Night Color=<rgba 0.7461 0.6737 1 1>
       Morning Color=<rgba 0.8728 0.9258 1 1>
       Day Color=<rgba 1 1 1 1>
       Evening Color=<rgba 0.8728 0.9258 1 1>
      ))
    ))
  ]
  particleUpdateScript [
   &SolveForcesAndVelocity(Clamp Velocity=false, Limit Acceleration=false)
   &FN_UpdateAnimTrail(Locked Particle RibbonWidth=1, SkeletalMesh=$System.Skeletal Mesh) {module=FN_UpdateAnimTrail001}
   &ScaleColor(
    Scale Alpha=&FloatFromCurve(
     FloatCurve=DataInterface() {propertyValues {Curve {keys [
         {interpMode=RCIM_Cubic, leaveTangent=-19.4136, tangentMode=RCTM_Break, time=0, value=1}
         {arriveTangent=-0.1476, tangentMode=RCTM_User, time=0.13, value=0}
        ]}}}
     Scale Curve=0.35
    )
    Scale RGB=&MakeVectorFromLinearColorRGB(Color=&ColorFromCurve(ColorCurve=DataInterface() {propertyValues {
        AlphaCurve {keys []}
        BlueCurve {keys [<at 0.0167 4 RCIM_Cubic>,<at 0.2127 1>,<at 0.5157 0.75 RCIM_Cubic>]}
        GreenCurve {keys [<at 0.0167 4 RCIM_Cubic>,<at 0.2127 1>,<at 0.5157 0.75 RCIM_Cubic>]}
        RedCurve {keys [<at 0.0167 4 RCIM_Cubic>,<at 0.2127 1>,<at 0.5157 0.75 RCIM_Cubic>]}
       }}))
    ScaleA=true
    ScaleRGB=true
   )
   &ScaleRibbonWidth(Ribbon Width Scale=&FloatFromCurve(FloatCurve=DataInterface() {propertyValues {Curve {keys [{interpMode=RCIM_Cubic, leaveTangent=-7.4157, tangentMode=RCTM_Break, time=0.2, value=1}, {tangentMode=RCTM_User, time=0.45, value=0}]}}}))
   &DynamicMaterialParameters(
    Index 0 Param 1=&FloatFromCurve(FloatCurve=DataInterface() {propertyValues {Curve {keys [<at 0 1>,<at 1 0>]}}})
    Index 0 Param 2=0.3
    Index 0 Param 3=2
    Index 0 Param 4=1
    Module.Index0Param0Write=true
    Module.Index0Param1Write=true
    Module.Index0Param2Write=true
    Module.Index0Param3Write=true
   )
   &ScaleColor(
    Scale Mode=::NewEnumerator1
    Scale RGBA=&MakeVector4_FromLinearColor(LinearColor=$System.ToDColor)
    ScaleRGBA=true
   ) {module=ScaleColor001}
  ]
  rendererClasses [&NiagaraRibbonRendererProperties]
  renderers [{rendererClass=&NiagaraRibbonRendererProperties}]
 }
}
summary {
 systemName=NS_Pickaxe_RankedBayonet_Trail
 emitters [
  {emitterName=SoftRibbon, rendererClasses [&NiagaraRibbonRendererProperties]}
  {emitterName=FillTrail, rendererClasses [&NiagaraRibbonRendererProperties]}
  {emitterName=FillTrail001, rendererClasses [&NiagaraRibbonRendererProperties]}
 ]
 userVariables [
  User.bIsFrontEndPreview {type=NiagaraBool, defaultValue=true}
  User.RankedProgression {type=NiagaraInt32}
  User.RibbonIDOffset {type=NiagaraInt32}
 ]
}
